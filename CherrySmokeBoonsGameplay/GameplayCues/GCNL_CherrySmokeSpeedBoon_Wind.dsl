(event EventOnLoopingStart(EitherVFX-preferthisversion)
  (params (MyTarget :type Actor%20Object%20Reference) (Parameters :type Gameplay%20Cue%20Parameters%20Structure%20%28by%20ref%29) (ParticleComponents :type Array%20of%20FXSystem%20Component%20Object%20References) (AudioComponents :type Array%20of%20Audio%20Component%20Object%20References))
  (bind _asplayer_pawn_athena (Utilities|Casting|CastToPlayerPawn_Athena MyTarget))
  (Variables|Default|SetTargetActor _asplayer_pawn_athena)
  (bind _asynctaskproxy (Ability|Tasks|WaitGameplayTagAddToActor (Variables|Default|GetTargetActor) (Variables|Default|GetCloakTag))
    (:then
      (Variables|Default|SetTag%20Add%20Watcher _asynctaskproxy)
      (bind _asynctaskproxy_2 (Ability|Tasks|WaitGameplayTagRemoveFromActor (Variables|Default|GetTargetActor) (Variables|Default|GetCloakTag))
        (:then
          (Variables|Default|SetTag%20Remove%20Watcher _asynctaskproxy_2)
          (bind _result (ArrayMacros|SafeArrayGet ParticleComponents)
            (:%20%20
              (bind _asynctaskproxy_3 (Ability|Tasks|WaitGameplayTagAddToActor (Variables|Default|GetTargetActor) "(TagName=\"Granted.Athena.Ability.TomatoSprinkle.Sprinting\")")
                (:then
                  (Variables|Default|SetTag%20Remove%20Watcher%20Sprint _asynctaskproxy_3))
                (:Added
                  (Utilities|FlowControl|DoOnce false
                    (:Completed
                      (Audio|SpawnSoundAttached "/CherrySmokeBoonsGameplay/Assets/Audio/SC_CherrySmokeBoons_Wind_A.SC_CherrySmokeBoons_Wind_A" (Variables|/Script/Engine.Character|GetMesh (Variables|Default|GetTargetActor)))
                      (Utilities|FlowControl|Delay 0))))))))
        (:Removed
          (bind _asniagara_particle_system_component (Utilities|Casting|CastToNiagaraComponent _result))
          (Rendering|SetVisibility _asniagara_particle_system_component true true))))
    (:Added
      (bind _asniagara_particle_system_component_2 (Utilities|Casting|CastToNiagaraComponent _result))
      (Rendering|SetVisibility _asniagara_particle_system_component_2 false true))))

(fn isPickaxeEquipped
  (return (Utilities|Enum|Equal(Enum) (Item|GetWeaponDataItemType (Weapon|GetCurrentWeapon (Variables|Default|GetTargetActor))) "WeaponHarvest")))

(fn ConstructionScript)
