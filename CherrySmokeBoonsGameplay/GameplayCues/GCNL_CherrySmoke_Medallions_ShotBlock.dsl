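; Looping gameplay cue for the CherrySmoke Medallions shot block, controlling a Sphere component on the cue.
; On loop start, tag add and remove tasks on the target track the headshot block cooldown and first-person camera as IsOnCooldown and IsADS.
; AdjustVis, called from that event, shows the sphere only when both flags are false.
; A note marks the first-person camera check as client only.
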(event EventOnLoopingStart(EitherVFX-preferthisversion)
  (params (MyTarget :type Actor%20Object%20Reference) (Parameters :type Gameplay%20Cue%20Parameters%20Structure%20%28by%20ref%29) (ParticleComponents :type Array%20of%20FXSystem%20Component%20Object%20References) (AudioComponents :type Array%20of%20Audio%20Component%20Object%20References))
  (Utilities|FlowControl|Sequence
    (:then_0
      (Ability|Tasks|WaitGameplayTagAddToActor MyTarget "(TagName=\"Ability.CherrySmoke.Medallions.HeadshotBlockCooldown\")"
        (:Added
          (Variables|Default|SetIsOnCooldown true))))
    (:then_1
      (Ability|Tasks|WaitGameplayTagRemoveFromActor MyTarget "(TagName=\"Ability.CherrySmoke.Medallions.HeadshotBlockCooldown\")"
        (:Removed
          (Variables|Default|SetIsOnCooldown false))))
    (:then_2
      ;[1 0.15 0.15 0.5] This has to be done client only
      (Ability|Tasks|WaitGameplayTagAddToActor MyTarget "(TagName=\"Gameplay.Camera.FirstPerson\")"
        (:Added
          (Variables|Default|SetIsADS true))))
    (:then_3
      (Ability|Tasks|WaitGameplayTagRemoveFromActor MyTarget "(TagName=\"Gameplay.Camera.FirstPerson\")"
        (:Removed
          (Variables|Default|SetIsADS false)))))
  (CallFunction|AdjustVis))

(event Custom|AdjustVis
  (Rendering|SetVisibility (Variables|Default|GetSphere) (not (or (Variables|Default|GetIsOnCooldown) (Variables|Default|GetIsADS))) true))

(fn ConstructionScript)
