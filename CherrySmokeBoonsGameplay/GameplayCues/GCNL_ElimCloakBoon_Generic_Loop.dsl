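; Looping gameplay cue that shows the Elim Cloak boon's cloak on a player pawn.
; Once cosmetics apply, it spawns Niagara effects and swaps in a timestamp-animated material override after a delay, reversing this on removal with allowance for SecretFacility invisibility.
; Tick holds cloak collection opacity at 1 and feeds smoothed speed to looping audio.
; Team visibility and interact handlers are empty; nothing here schedules the expiry warning.
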(event EventOnLoopingStart(EitherVFX-preferthisversion)
  (params (MyTarget :type Actor%20Object%20Reference) (Parameters :type Gameplay%20Cue%20Parameters%20Structure%20%28by%20ref%29) (ParticleComponents :type Array%20of%20FXSystem%20Component%20Object%20References) (AudioComponents :type Array%20of%20Audio%20Component%20Object%20References))
  ;[0 1 0.05708] Looping Start
  (|CallParentFunction MyTarget Parameters ParticleComponents AudioComponents)
  (Utilities|FlowControl|Sequence
    (:then_0
      ;;[0 1 0.09341] Initialze variables
      (bind _asfort_player_pawn (Utilities|Casting|CastToFortPlayerPawn MyTarget)
        (:then
          (Variables|Default|SetTargetPlayer _asfort_player_pawn))
        (:CastFailed
          _))
      (Actor|Tick|SetActorTickEnabled true)
      (Variables|Default|SetStationaryVisMult (ScalableFloat|GetValueatLevel (Variables|Balance|GetRow_StationaryVisMult)))
      (Variables|Default|SetMaxSpeedVisMult (ScalableFloat|GetValueatLevel (Variables|Balance|GetRow_MaxSpeedVisMult)))
      (Variables|Default|SetVisibilityMinFriendly (ScalableFloat|GetValueatLevel (Variables|Balance|GetRow_FriendlyVisMultMin)))
      (Variables|Default|SetVisibilityMinNonfriendly (ScalableFloat|GetValueatLevel (Variables|Balance|GetRow_NonfriendlyVisMultMin)))
      (Variables|Balance|CalculatedValues|SetInteractVisMultAddMax (ScalableFloat|GetValueatLevel (Variables|Balance|GetRow_InteractVisMultAddMax)))
      (Variables|Default|SetVisibilityLevelChangeRate (ScalableFloat|GetValueatLevel (Variables|Balance|GetRow_VisibilityLevelChangeRate))))
    (:then_3)
    (:then_4
      ;;[0 0.9642 1] Set-up for the local player
      (Rendering|Material|SetScalarParameterValue 0 "/CherrySmokeBoonsGameplay/Gameplay/ElimCloak/Materials/MPC_Cloak_Generic.MPC_Cloak_Generic" "Opacity" 1.0 :!class /Script/Engine.KismetMaterialLibrary))
    (:then_5
      ;;[0.7159 1 0] Initialize the skeletal meshes
      (CallFunction|InitializePlayerSkeletalMeshes))))

(event Custom|Initialize%20Player%20Skeletal%20Meshes
  (bind _targetplayer (Variables|Default|GetTargetPlayer))
  ;[0.4358 1 0] Initialize Player Skeletal Meshes
  (Utilities|DoOncePerFrame
    (:%20%20
      (Utilities|Array|Clear (Variables|Default|GetPlayerSkeletalMeshes))
      (Utilities|IsValid _targetplayer
        (:Is%20Valid
          (bind _delegatehandlecontroller_out (Cosmetics|WhenCosmeticsApplied _targetplayer (delegate Custom|OnCosmeticsApplied)))
          (Variables|Default|SetDelegate%20Handle%20Controller _delegatehandlecontroller_out))))))

(event Custom|OnCosmeticsApplied
  (params (Pawn :type Fort%20Player%20Pawn%20Object%20Reference) (BaseMesh :type Skeletal%20Mesh%20Component%20Object%20Reference) (PartsArray :type Array%20of%20Custom%20Character%20Part%20Object%20References) (ValidSkeletalMeshComponentPartArray :type Array%20of%20Skeletal%20Mesh%20Component%20Object%20References))
  (bind _playerskeletalmeshes (Variables|Default|GetPlayerSkeletalMeshes))
  ;[0.4358 1 0] Initialize Player Skeletal Meshes
  ;;[0.1083 1 0.0636] OnCosmeticsApplied: Collect updated meshes and try to apply the cloak material to them.
  (Utilities|FlowControl|Sequence
    (:then_0
      ;;;[0.02461 0 1] Populate PlayerSkeletalMeshes
      (Utilities|Array|Clear _playerskeletalmeshes)
      (for _array_element PartsArray
        (Utilities|IsValid _array_element
          (:Is%20Valid
            (Utilities|Array|Add _playerskeletalmeshes)))))
    (:then_1
      ;;;[1 0.5431 0] If we do NOT have skeletal meshes yet, then schedule a retry.
      ;;; If we do have skeletal meshes then set those meshes to use the cloaked material.
      (if (<= (Utilities|Array|Length _playerskeletalmeshes) 0)
        (Utilities|Time|SetTimerforNextTickbyEvent (EventDispatchers|CreateEvent :!event Custom|Initialize%20Player%20Skeletal%20Meshes))
        (else
          (CallFunction|SetMaterialCloaked))))))

(event EventOnRemoval(EitherVFX-preferthisversion)
  (params (MyTarget :type Actor%20Object%20Reference) (Parameters :type Gameplay%20Cue%20Parameters%20Structure%20%28by%20ref%29) (ParticleComponents :type Array%20of%20FXSystem%20Component%20Object%20References) (AudioComponents :type Array%20of%20Audio%20Component%20Object%20References) (BurstCameraShakeInstance :type Camera%20Shake%20Base%20Object%20Reference) (BurstDecalInstance :type Decal%20Actor%20Object%20Reference))
  (bind _targetplayer (Variables|Default|GetTargetPlayer))
  ;[1 0.4543 0] Handle Removal
  (|CallParentFunction MyTarget Parameters ParticleComponents AudioComponents BurstCameraShakeInstance BurstDecalInstance)
  (bind _asfort_player_pawn (Utilities|Casting|CastToFortPlayerPawn MyTarget)
    (:then
      (Variables|Default|SetTargetPlayer _asfort_player_pawn))
    (:CastFailed
      _))
  (Actor|Tick|SetActorTickEnabled false)
  (CallFunction|ClearExpireTell)
  (Utilities|IsValid _targetplayer
    (:Is%20Valid
      (EventDispatchers|UnbindEventfromOnStartedInteractSearch _targetplayer (EventDispatchers|CreateEvent :!event Custom|StartedInteract))
      (EventDispatchers|UnbindEventfromOnStartedInteractSearch _targetplayer (EventDispatchers|CreateEvent :!event Custom|EndedInteract))
      (CallFunction|SetMaterialUncloaked))))

(event EventTick
  (params (DeltaSeconds :type Float%20%28single-precision%29))
  ;[0.07843 0.4235 1] Event tick
  (Variables|Default|GetTargetPlayer
    (:then
      (Utilities|FlowControl|Sequence
        (:then_0
          ;;[0 1 1] Checks for what the player is doing. Determines the material opacity
          (Rendering|Material|SetScalarParameterValue 0 "/CherrySmokeBoonsGameplay/Gameplay/ElimCloak/Materials/MPC_Cloak_Generic.MPC_Cloak_Generic" "Opacity" 1.0 :!class /Script/Engine.KismetMaterialLibrary)
          (if (Variables|Default|GetDebugLogging%20Const)
            (Development|PrintString 0 (Utilities|String|Append "GCNL_Cloak_Generic_Loop::Tick() TargetPlayer: [" (Fortnite|GetPlayerNameSafe (Variables|Default|GetTargetPlayer)) "], Current Opacity: [" (Utilities|String|ToString(Float) 1.0) "].") true true "(R=0.000000,G=0.660000,B=1.000000,A=1.000000)" 0.0)))
        (:then_1
          ;;[0.05415 1 0.7377] Update Team Visibility
          (CallFunction|UpdateTeamVisbility))
        (:then_2
          ;;[0.6823 0.3646 1] Perform the Audio Tick
          (if (> (Utilities|Array|Length (Variables|Default|GetLoopingAudioComponents)) 0)
            (CallFunction|AudioTick (Math|Vector|VectorLength (Transformation|GetVelocity (Variables|Default|GetTargetPlayer))))))))
    (:else)))

(event Custom|Audio%20Tick
  (params (Velocity :type Float%20%28double-precision%29))
  (bind _loopingaudiocomponents (Variables|Default|GetLoopingAudioComponents))
  (bind _output (Utilities|Array|Get(acopy) _loopingaudiocomponents))
  ;[0.2431 0.2588 1] Audio Tick
  (if (Utilities|Array|IsValidIndex _loopingaudiocomponents)
    (Utilities|IsValid _output
      (:Is%20Valid
        (bind _output_get (Variables|Balance|CalculatedValues|SetAudioVelocityInterp (Math|Interpolation|FInterpTo (Variables|Balance|CalculatedValues|GetAudioVelocityInterp) (Math|Float|MapRangeClamped Velocity 6e+02 1e+03 0.0 1.0) (Utilities|Time|GetWorldDeltaSeconds 0) 1.5)))
        (Audio|Parameter|SetFloatParameter _output (Variables|ParameterName|GetAudioParameterName_Velocity) _output_get :!class /Script/AudioExtensions.AudioParameterControllerInterface)))))

(event Custom|StartedInteract)

(event Custom|EndedInteract)

(event Custom|UpdateTeamVisbility)

(event Custom|BeginExpireTell)

(event Custom|ExpireWarningTick
  (bind _expirationsoundperiodscale (Variables|Balance|GetExpirationSoundPeriodScale))
  (bind _expirationsoundperiod (Variables|Balance|GetExpirationSoundPeriod))
  (bind _returnvalue (* _expirationsoundperiod _expirationsoundperiodscale))
  (bind _returnvalue_2 (/ _returnvalue 2.0))
  (bind _returnvalue_3 (Utilities|Time|GetGameTimeinSeconds 0))
  (bind _targetplayer (Variables|Default|GetTargetPlayer))
  ;[0.543 0 1] Expiration Warning Sequence
  (Audio|SpawnSoundAttached (Variables|Default|GetEnd%20Warning%20Sound%20Cue) (Variables|/Script/Engine.Actor|GetRootComponent (Variables|Default|GetTargetPlayer)))
  (Class|FortPlayerPawn|SetScalarParamonOverriddenMaterials _targetplayer "ExpireTimeStampA" _returnvalue_3)
  (Class|FortPlayerPawn|SetScalarParamonOverriddenMaterials _targetplayer "ExpireTimeStampB" (+ _returnvalue_3 _returnvalue_2))
  (Class|FortPlayerPawn|SetScalarParamonOverriddenMaterials _targetplayer "ExpireTransitionLength" _returnvalue_2))

(event Custom|ClearExpireTell
  ;[0.543 0 1] Expiration Warning Sequence
  (Utilities|Time|ClearandInvalidateTimerbyHandle 0 (Variables|Default|GetExpireTellDelayTimer))
  (Utilities|Time|ClearandInvalidateTimerbyHandle 0 (Variables|Default|GetExpirationSoundTimer)))

(event Custom|SetMaterialCloaked
  ;[1 1 1] Cloak Activate
  (Utilities|Time|ClearandInvalidateTimerbyHandle 0 (Variables|FX|GetRemoveCloakMaterialOverrideHandle))
  (bind _targetplayer (Variables|Default|GetTargetPlayer)
    (:then
      (bind _returnvalue (Niagara|SpawnSystemAttached (Variables|FX|GetActivateFXTemplate) (Variables|/Script/Engine.Character|GetMesh _targetplayer) "root" "0, 0, 0" "0, 0, 0" "KeepRelativeOffset" true true "AutoRelease"))
      (bind _output_get (Variables|Default|SetActivateCloakFX _returnvalue))
      (Effects|Components|ParticleSystem|SetActorParameter _returnvalue "SkelMesh" (Variables|Default|GetTargetPlayer))
      (bind _returnvalue_2 (Utilities|Time|SetTimerbyFunctionName 0 "ApplyCloakMaterialOverride" (Variables|FX|GetDelayBeforeApplyingCloakMaterialOverride)))
      (Variables|FX|SetApplyCloakMaterialOverrideHandle _returnvalue_2))
    (:else)))

(event Custom|SetMaterialUncloaked
  (bind _targetplayer (Variables|Default|GetTargetPlayer))
  (bind _returnvalue (GameplayTags|HasMatchingGameplayTag _targetplayer "(TagName=\"SecretFacility.Invisibility.Enabled\")"))
  (bind _returnvalue_2 (not _returnvalue))
  (bind _targetplayer_2 (Variables|Default|GetTargetPlayer))
  ;[1 1 1] Cloak Remove
  (Utilities|Time|ClearandInvalidateTimerbyHandle 0 (Variables|FX|GetApplyCloakMaterialOverrideHandle))
  (if _returnvalue_2
    (Class|FortPlayerPawn|SetScalarParamonOverriddenMaterials _targetplayer_2 "TimeStampA" 0.0)
    (Class|FortPlayerPawn|SetScalarParamonOverriddenMaterials _targetplayer_2 "TimeStampB" (Utilities|Time|GetGameTimeinSeconds 0))
    (else
      _))
  (bind _returnvalue_3 (Utilities|Time|SetTimerbyFunctionName 0 "RemoveCloakMaterialOverride" (Math|Float|SelectFloat (Variables|FX|GetDelayBeforeRemovingCloakMaterialOverride) (* (Variables|FX|GetDelayBeforeRemovingCloakMaterialOverride) 1.5) _returnvalue_2)))
  (Variables|FX|SetRemoveCloakMaterialOverrideHandle _returnvalue_3))

(fn ApplyCloakMaterialOverride
  (bind _targetplayer (Variables|Default|GetTargetPlayer))
  (Rendering|SetCastShadow (Variables|/Script/Engine.Character|GetMesh (Variables|Default|GetTargetPlayer)) false)
  (bind _returnvalue (Class|FortPlayerPawn|ApplyMaterialOverrideLocalwithFlags (Variables|Default|GetTargetPlayer) (Utilities|ResolveSoftReference (Variables|FX|GetM_Cloaked)) (Variables|Default|GetMaterialParametersToCopy) 0.0 3))
  (Variables|Default|SetMaterialOverrideId _returnvalue)
  (Class|FortPlayerPawn|SetScalarParamonOverriddenMaterials _targetplayer "TimeStampA" (Utilities|Time|GetGameTimeinSeconds 0))
  (Class|FortPlayerPawn|SetScalarParamonOverriddenMaterials _targetplayer "TimeStampB" 1e+10))

(fn IsPlayerControllerFriendly
  (params (PlayerController :type Controller%20Object%20Reference))
  (Utilities|IsValid PlayerController
    (:Is%20Valid
      (bind _targetplayer (Variables|Default|GetTargetPlayer)
        (:then
          ;[1 0.4635 0.4635] Both players are valid, return if they are on the same team
          (return (Team|OnSameTeam (Variables|/Script/Engine.Pawn|GetPlayerState _targetplayer) (Variables|/Script/Engine.Controller|GetPlayerState PlayerController))))
        (:else
          _)))
    (:Is%20Not%20Valid
      _))
  ;[1 0.4635 0.4635] One or both players is nullptr, return FALSE
  (return false))

(fn RemoveCloakMaterialOverride
  (Utilities|FlowControl|Sequence
    (:then_0
      (Class|FortPlayerPawn|RemoveMaterialOverrideLocal (Variables|Default|GetTargetPlayer) (Variables|Default|GetMaterialOverrideId))
      (Guid|InvalidateGuid (Variables|Default|GetMaterialOverrideId))
      (bind _targetplayer (Variables|Default|GetTargetPlayer)
        (:then
          (Niagara|SpawnSystemAttached (Variables|FX|GetDeactivateFXTemplate) (Variables|/Script/Engine.Character|GetMesh _targetplayer) "root" "0, 0, 0" "0, 0, 0" "KeepRelativeOffset" true true "AutoRelease"))
        (:else)))
    (:then_1
      (if (not (GameplayTags|HasMatchingGameplayTag (Variables|Default|GetTargetPlayer) "(TagName=\"SecretFacility.Invisibility.Enabled\")"))
        (Rendering|SetCastShadow (Variables|/Script/Engine.Character|GetMesh (Variables|Default|GetTargetPlayer)) true)
        (Variables|Default|SetTargetPlayer 0)))))

(fn ShouldRenderAsLocalPlayer
  (bind _targetplayer (Variables|Default|GetTargetPlayer)
    (:then
      ;[1 0.4896 0.4896] Return if this GC shourender as if this is the local player
      (return (Math|Boolean|ORBoolean (Pawn|IsLocallyControlled _targetplayer) (Replay|GetIsPlayingReplay 0) (Broadcast|GetIsBroadcastClient 0))))
    (:else
      ;[1 0.4896 0.4896] Could not find Target Player, Return FALSE
      (return false))))

(fn ConstructionScript)
