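; Looping gameplay cue notify for the elimination cloak boon that applies a cloaked highlight to the instigating player's pawn.
; OnBurst resolves the pawn through the instigator's FortPlayerState, adds a custom highlighting override with configurable data and priority, and stores the returned handle.
; OnCeaseRelevant resolves the pawn the same way and uses that handle to remove the override.
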(event EventBeginPlay)

(event Collision|EventActorBeginOverlap
  (params (OtherActor :type Actor%20Object%20Reference)))

(event EventTick
  (params (DeltaSeconds :type Float%20%28single-precision%29))
  (|CallParentFunction DeltaSeconds) ; This node is disabled and will not be called.\nDrag off pins to build functionality.
)

(fn OnBurst
  (params (MyTarget :type Actor%20Object%20Reference) (Parameters :type Gameplay%20Cue%20Parameters%20Structure%20%28by%20ref%29))
  (bind _returnvalue (|CallParentFunction MyTarget Parameters))
  (bind _asfort_player_state (Utilities|Casting|CastToFortPlayerState (Ability|GameplayCue|GetInstigatorActor Parameters)))
  (bind _asfort_player_pawn_athena (Utilities|Casting|CastToFortPlayerPawnAthena (PlayerState|GetPawn _asfort_player_state)))
  (bind _returnvalue_2 (Highlighting|AddCustomHighlightingDataOverride _asfort_player_pawn_athena (Variables|Default|GetCloakedCustomHighlightingData) (Variables|Default|GetCloakingHighlightOverridePriority)))
  (Variables|Default|SetCloakedCustomHighlightingOverrideHandle _returnvalue_2)
  (return _returnvalue))

(fn OnCeaseRelevant
  (params (MyTarget :type Actor%20Object%20Reference) (Parameters :type Gameplay%20Cue%20Parameters%20Structure%20%28by%20ref%29))
  (bind _returnvalue (|CallParentFunction MyTarget Parameters))
  (bind _asfort_player_state (Utilities|Casting|CastToFortPlayerState (Ability|GameplayCue|GetInstigatorActor Parameters)))
  (bind _asfort_player_pawn_athena (Utilities|Casting|CastToFortPlayerPawnAthena (PlayerState|GetPawn _asfort_player_state)))
  (Highlighting|RemoveCustomHighlightingDataOverride _asfort_player_pawn_athena (Variables|Default|GetCloakedCustomHighlightingOverrideHandle))
  (return _returnvalue))

(fn ConstructionScript
  (|CallParentFunction))
